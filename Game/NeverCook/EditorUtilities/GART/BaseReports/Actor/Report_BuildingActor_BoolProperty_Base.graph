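# Base GART editor report that validates a configurable bool property on BuildingActors against an expected value.
# RunFilterLogicOnObject casts the object to BuildingActor and loops through a parent-class whitelist, breaking on a match.
# It reads the named property via GetEditorProperty from the actor or, when flagged, the same input object cast to StaticMeshComponent.
# It returns whether the value matches, with report data built from configurable description and severity.
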EditorUtilityBlueprint /Game/NeverCook/EditorUtilities/GART/BaseReports/Actor/Report_BuildingActor_BoolProperty_Base
graphs [
 EventGraph {links [], nodes []}
 RunFilterLogicOnObject {nodes [
   n1 = RunFilterLogicOnObject!() -> (then, object:Object Reference) {is=title K2Node_FunctionEntry}
   > n2 = Cast To BuildingActor!(execute, Object=@n1) -> (then, CastFailed, AsBuilding Actor:Building Actor Object Reference) {is=title K2Node_DynamicCast}
   > n3 = Set l_buildingActor(l_buildingActor:Building Actor Object Reference=@n2) -> (Output_Get:Building Actor Object Reference)
   > n4 = For Each Loop with Break!(
    Exec:Exec
    Array:Array of Object Soft Class References=Get parentClassWhitelist() -> (parentClassWhitelist:Array of Object Soft Class References)
    Break:Exec=@n16.then
   ) -> (LoopBody:Exec, Array Element:Object Soft Class Reference, Array Index:int, Completed:Exec)
   n5 = ClassIsChildOf(
    TestClass=Get Class(Object=@n3) -> (ReturnValue:Building Actor Class Reference)
    ParentClass=Resolve Soft Reference!(Input:Object Soft Class Reference=@n4) -> (Output:Object Class Reference)
   )
   n6 = Branch(execute=@n4.Completed, Condition=@n5)
   > n7 = Branch(Condition=Get isSM_ComponentProperty() -> (isSM_ComponentProperty:bool))
   > n8 = Cast To StaticMeshComponent(Object=Get object() -> (object:Object Reference))
   > n9 = GetEditorProperty!(execute:Exec, Object:Object Reference=@n8, PropertyName:Name=Get propertyName() -> (propertyName:Name=None)) -> (then:Exec, PropertyValue:bool, ReturnValue:bool)
   > n10 = Branch(Condition=@n9.ReturnValue)
   n11 = Get propertyName() -> (propertyName=None)
   n12 = NOT Boolean(A=Get expectedResult() -> (expectedResult:bool))
   n13 = Make Struct GART Report Data(
    DisplayTitle_20_#1=Building Actor
    DisplayDescShort_29_#2=Append!(A="To String (Name)"(InName=@n11), B=., C:String="To String (Boolean)"(InBool=@n12)) -> (ReturnValue)
    DisplayDescLong_30_#3=Get reportDescription() -> (reportDescription:String)
    Severity_46_#4=Get reportSeverity() -> (reportSeverity:Enum_GART_ReportSeverity Enum)
   )
   > Return Node!(
    execute
    Valid:bool="Equal (Boolean)"(A:bool=Get expectedResult(), B:bool=@n9) -> (ReturnValue:bool=~) {is=title K2Node_PromotableOperator}
    ReportData:Struct GART Report Data Structure=@n13
   )
   Return Node(execute=@n10.else)
   n14 = GetEditorProperty(
    execute=@n7.else
    Object=Get l_buildingActor() -> (l_buildingActor:Building Actor Object Reference)
    PropertyName=Get propertyName() -> (propertyName=None)
   )
   > n15 = Branch(Condition=@n14.ReturnValue)
   > Return Node(Valid="Equal (Boolean)"(A=Get expectedResult(), B=@n14) -> (ReturnValue=~), ReportData=@n13)
   Return Node(execute=@n15.else)
   Return Node(execute=@n6.else)
   n16 = Branch(execute=@n4.LoopBody, Condition=@n5)
   Return Node(execute=@n2.CastFailed)
  ]}
]
variables [
 expectedResult {type=boolean}
 isSM_ComponentProperty {type=boolean}
 parentClassWhitelist {type="array<Class>", default []}
 propertyName {type=string, default=None}
 reportDescription {type=string, default=""}
 reportSeverity {type=Enum_GART_ReportSeverity, default=Note}
]
